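# One-shot stereo MetaSound, named as a preset, that plays a random Radical CoreAR pickup sound with item attenuation and concurrency.
# On play, two random-get and wave-player chains feed a mixer, one drawing on three pickup waves, the other on an array variable showing no assets.
# Finish fires through an accumulator on both players.
# An MSS_Item_Pickup node receives the volume, pitch-shift and sound inputs, but none of its outputs are linked.
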MetaSoundSource /RadicalWeaponsGameplay/Weapons/RadicalCoreAR/Audio/MSS_Radical_CoreAR_Athena_Pickup_Preset
inputs [
 MasterVolume {type=Float}
 Random.PitchShift.Base {type=Float}
 Random.PitchShift.Spread {type=Float}
 Sounds {type=WaveAsset:Array}
 UE.Source.OnPlay {type=Trigger}
]
interfaces [UE.OutputFormat.Stereo:1.0, UE.Source.OneShot:1.0, UE.Source:1.1]
outputs [
 "UE.OutputFormat.Stereo.Audio:0" {type=Audio}
 "UE.OutputFormat.Stereo.Audio:1" {type=Audio}
 UE.Source.OneShot.OnFinished {type=Trigger}
]
pages [Default {
  input_defaults {
   MasterVolume=0.5
   Random.PitchShift.Base=0
   Random.PitchShift.Spread=0.32
   UE.Source.OnPlay=null
   Sounds [
    /RadicalWeaponsGameplay/Weapons/RadicalCoreAR/Audio/SoundWaves/Radical_CoreAR_Athena_Pickup_01
    /RadicalWeaponsGameplay/Weapons/RadicalCoreAR/Audio/SoundWaves/Radical_CoreAR_Athena_Pickup_02
    /RadicalWeaponsGameplay/Weapons/RadicalCoreAR/Audio/SoundWaves/Radical_CoreAR_Athena_Pickup_03
   ]
  }
  nodes [
   ArrayRandomGet_0 = Array.Random Get:WaveAsset:Array(In Array=^Sounds, Next=^UE.Source.OnPlay) {also {defaults {Weights []}}}
   WavePlayer_0 = UE.Wave Player:Stereo(Play=@ArrayRandomGet_0.On Next, Wave Asset=@ArrayRandomGet_0.Value) {also {version [1,1]}}
   ArrayRandomGet_1 = Array.Random Get:WaveAsset:Array(Next=^UE.Source.OnPlay, In Array=VariableAccessor.WaveAsset:Array(Variable=InitVariable.WaveAsset:Array().Variable).Value)
   WavePlayer_1 = UE.Wave Player:Stereo(Play=@ArrayRandomGet_1.On Next, Wave Asset=@ArrayRandomGet_1.Value) {also {version [1,1]}}
   AudioMixerAudioMixerStereo2_0 = "AudioMixer.Audio Mixer (Stereo, 2)"(In 1 L=@WavePlayer_0.Out Left, In 1 R=@WavePlayer_0.Out Right, In 0 L=@WavePlayer_1.Out Left, In 0 R=@WavePlayer_1.Out Right)
   /FortMetasound/Items/MSS_Item_Pickup(
    MasterVolume=^MasterVolume
    Random.PitchShift.Base=^Random.PitchShift.Base
    Random.PitchShift.Spread=^Random.PitchShift.Spread
    Sounds=^Sounds
    UE.Source.OnPlay=^UE.Source.OnPlay
   )
   TriggerAccumulatorTriggerAccumulate2_0 = "TriggerAccumulator.Trigger Accumulate (2)"(In 1=@WavePlayer_0.On Finished, In 0=@WavePlayer_1.On Finished)
  ]
  links [
   {dst="@UE.OutputFormat.Stereo.Audio:0", src=AudioMixerAudioMixerStereo2_0.Out L}
   {dst="@UE.OutputFormat.Stereo.Audio:1", src=AudioMixerAudioMixerStereo2_0.Out R}
   {dst="@UE.Source.OneShot.OnFinished", src=TriggerAccumulatorTriggerAccumulate2_0.Out}
  ]
 }]
properties {
 assetImportData=&self:AssetImportData
 attenuationSettings=/FortMetasound/Items/AS_Fort_Item
 maxDistance=3500
 numChannels=2
 outputFormat=Stereo
 sampleRate=48000
 soundClassObject=&Fort_Loot_Collect
 volume=0.2
 concurrencySet [&CS_Fort_Item]
}
