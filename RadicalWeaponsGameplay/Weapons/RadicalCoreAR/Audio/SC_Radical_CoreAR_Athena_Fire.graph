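# Sound cue for the Radical Core AR's fire sound in Athena, classed as third-person ranged weapon fire.
# A mixer layers randomized close shots with quieter-mixed, randomized distant shots, each branch using its own attenuation and sound class override, under a root modulator.
# The distant branch gets a modulator volume boost, and both random nodes are excluded from branch culling.
# Global and local gunfire concurrency sets govern simultaneous playback.
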SoundCue /RadicalWeaponsGameplay/Weapons/RadicalCoreAR/Audio/SC_Radical_CoreAR_Athena_Fire
root=Modulator_3
nodes [
 {id=Attenuation_3, children [Random_3], properties {attenuationSettings=&Fort_Weapon_Projectile_Close_Athena_Att}}
 {
  id=Attenuation_5
  children [Modulator_6]
  properties {attenuationSettings=&Fort_Weapon_Projectile_Medium_Mono_Athena_Att}
 }
 {id=Mixer_3, children [SoundClass_3,SoundClass_5], properties {inputVolume [0.7,0.18]}}
 {id=Modulator_3, children [Mixer_3]}
 {id=Modulator_6, children [Random_6], properties {volumeMax=2, volumeMin=2}}
 {
  id=Random_3
  children [WavePlayer_0,WavePlayer_1,WavePlayer_2,WavePlayer_7,WavePlayer_8,WavePlayer_9]
  properties {bShouldExcludeFromBranchCulling=true, weights [1,1,1,1,1,1]}
 }
 {
  id=Random_6
  children [WavePlayer_22, WavePlayer_23, WavePlayer_28, WavePlayer_29]
  properties {bShouldExcludeFromBranchCulling=true, weights [1,1,1,1]}
 }
 {
  id=SoundClass_3
  children [Attenuation_3]
  properties {soundClassOverride=&Fort_WeaponRanged_Fire_3P_PSM}
 }
 {
  id=SoundClass_5
  children [Attenuation_5]
  properties {soundClassOverride=&Fort_WeaponRanged_Fire_Distant}
 }
 {
  id=WavePlayer_0
  properties {soundWaveAssetPtr=/RadicalWeaponsGameplay/Weapons/RadicalCoreAR/Audio/SoundWaves/Radical_CoreAR_Athena_Fire_Close_01}
 }
 {
  id=WavePlayer_1
  properties {soundWaveAssetPtr=/RadicalWeaponsGameplay/Weapons/RadicalCoreAR/Audio/SoundWaves/Radical_CoreAR_Athena_Fire_Close_02}
 }
 {
  id=WavePlayer_2
  properties {soundWaveAssetPtr=/RadicalWeaponsGameplay/Weapons/RadicalCoreAR/Audio/SoundWaves/Radical_CoreAR_Athena_Fire_Close_03}
 }
 {
  id=WavePlayer_22
  properties {soundWaveAssetPtr=/RadicalWeaponsGameplay/Weapons/RadicalCoreAR/Audio/SoundWaves/Radical_CoreAR_Athena_Fire_Dist_01}
 }
 {
  id=WavePlayer_23
  properties {soundWaveAssetPtr=/RadicalWeaponsGameplay/Weapons/RadicalCoreAR/Audio/SoundWaves/Radical_CoreAR_Athena_Fire_Dist_02}
 }
 {
  id=WavePlayer_28
  properties {soundWaveAssetPtr=/RadicalWeaponsGameplay/Weapons/RadicalCoreAR/Audio/SoundWaves/Radical_CoreAR_Athena_Fire_Dist_03}
 }
 {
  id=WavePlayer_29
  properties {soundWaveAssetPtr=/RadicalWeaponsGameplay/Weapons/RadicalCoreAR/Audio/SoundWaves/Radical_CoreAR_Athena_Fire_Dist_04}
 }
 {
  id=WavePlayer_7
  properties {soundWaveAssetPtr=/RadicalWeaponsGameplay/Weapons/RadicalCoreAR/Audio/SoundWaves/Radical_CoreAR_Athena_Fire_Close_04}
 }
 {
  id=WavePlayer_8
  properties {soundWaveAssetPtr=/RadicalWeaponsGameplay/Weapons/RadicalCoreAR/Audio/SoundWaves/Radical_CoreAR_Athena_Fire_Close_05}
 }
 {
  id=WavePlayer_9
  properties {soundWaveAssetPtr=/RadicalWeaponsGameplay/Weapons/RadicalCoreAR/Audio/SoundWaves/Radical_CoreAR_Athena_Fire_Close_06}
 }
]
properties {
 bDebug=true
 maxDistance=10579.2871
 soundClassObject=&Fort_WeaponRanged_Fire_3P_PSM
 volumeMultiplier=0.9
 concurrencyOverrides {bLimitToOwner=true, maxCount=3, resolutionRule=StopOldest}
 concurrencySet [&Weapon_Gunfire_Global_Conc, &Weapon_Gunfire_Local_Conc]
}
