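# Sound cue for the low-ammo sound of the Radical Core AR weapon.
# A Random node picks among six equally weighted LowAmmo wave variations, and a Modulator applies very slight pitch and volume variation.
# The cue plays at reduced volume in a ranged-weapon low-ammo sound class.
# It overrides concurrency with a PreventNew rule, so new instances are refused once the limit is reached.
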SoundCue /RadicalWeaponsGameplay/Weapons/RadicalCoreAR/Audio/SC_Radical_CoreAR_Athena_LowAmmo
root=Modulator_0
nodes [
 {id=Modulator_0, children [Random_0], properties {pitchMax=1, pitchMin=0.99, volumeMax=1.01, volumeMin=0.99}}
 {id=Random_0, children [WavePlayer_0,WavePlayer_1,WavePlayer_2,WavePlayer_3,WavePlayer_4,WavePlayer_5], properties {weights [1,1,1,1,1,1]}}
 {
  id=WavePlayer_0
  properties {soundWaveAssetPtr=/RadicalWeaponsGameplay/Weapons/RadicalCoreAR/Audio/SoundWaves/Radical_CoreAR_Athena_LowAmmo_01}
 }
 {
  id=WavePlayer_1
  properties {soundWaveAssetPtr=/RadicalWeaponsGameplay/Weapons/RadicalCoreAR/Audio/SoundWaves/Radical_CoreAR_Athena_LowAmmo_02}
 }
 {
  id=WavePlayer_2
  properties {soundWaveAssetPtr=/RadicalWeaponsGameplay/Weapons/RadicalCoreAR/Audio/SoundWaves/Radical_CoreAR_Athena_LowAmmo_03}
 }
 {
  id=WavePlayer_3
  properties {soundWaveAssetPtr=/RadicalWeaponsGameplay/Weapons/RadicalCoreAR/Audio/SoundWaves/Radical_CoreAR_Athena_LowAmmo_04}
 }
 {
  id=WavePlayer_4
  properties {soundWaveAssetPtr=/RadicalWeaponsGameplay/Weapons/RadicalCoreAR/Audio/SoundWaves/Radical_CoreAR_Athena_LowAmmo_05}
 }
 {
  id=WavePlayer_5
  properties {soundWaveAssetPtr=/RadicalWeaponsGameplay/Weapons/RadicalCoreAR/Audio/SoundWaves/Radical_CoreAR_Athena_LowAmmo_06}
 }
]
properties {
 bOverrideConcurrency=true
 maxDistance=2097152
 soundClassObject=/Game/Sounds/FortSoundClassesAndModes/Fort_WeaponRanged_LowAmmo_Degraded_PSM
 volumeMultiplier=0.6
 concurrencyOverrides {resolutionRule=PreventNew}
}
