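# Sound cue for the Radical CoreAR weapon's no-ammo sound.
# A Random node picks one of three equally weighted wave variants, and a Modulator applies slight volume variation at fixed pitch.
# It routes to the third-person ranged weapon fire sound class at reduced cue volume, with a concurrency override that caps simultaneous plays and stops the oldest.
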SoundCue /RadicalWeaponsGameplay/Weapons/RadicalCoreAR/Audio/SC_Radical_CoreAR_Athena_NoAmmo
root=Modulator_0
nodes [
 {id=Modulator_0, children [Random_0], properties {pitchMax=1, pitchMin=1, volumeMax=0.95, volumeMin=0.85}}
 {id=Random_0, children [WavePlayer_4,WavePlayer_5,WavePlayer_6], properties {weights [1,1,1]}}
 {
  id=WavePlayer_4
  properties {soundWaveAssetPtr=/RadicalWeaponsGameplay/Weapons/RadicalCoreAR/Audio/SoundWaves/Radical_CoreAR_Athena_NoAmmo_01}
 }
 {
  id=WavePlayer_5
  properties {soundWaveAssetPtr=/RadicalWeaponsGameplay/Weapons/RadicalCoreAR/Audio/SoundWaves/Radical_CoreAR_Athena_NoAmmo_02}
 }
 {
  id=WavePlayer_6
  properties {soundWaveAssetPtr=/RadicalWeaponsGameplay/Weapons/RadicalCoreAR/Audio/SoundWaves/Radical_CoreAR_Athena_NoAmmo_03}
 }
]
properties {
 bOverrideConcurrency=true
 maxDistance=2097152
 soundClassObject=&Fort_WeaponRanged_Fire_3P_PSM
 volumeMultiplier=0.55
 attenuationOverrides {falloffDistance=1600}
 concurrencyOverrides {maxCount=2, resolutionRule=StopOldest}
}
