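# Sound cue for the start of the Radical Core AR reload sound.
# A random node picks one of three equally weighted wave variations.
# It plays as weapon foley at reduced volume, using shared weapon reload attenuation.
# The cue overrides concurrency and has its debug flag switched on.
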SoundCue /RadicalWeaponsGameplay/Weapons/RadicalCoreAR/Audio/SC_Radical_CoreAR_Athena_Reload_Start
root=Random_0
nodes [
 {id=Random_0, children [WavePlayer_0,WavePlayer_1,WavePlayer_2], properties {weights [1,1,1]}}
 {
  id=WavePlayer_0
  properties {soundWaveAssetPtr=/RadicalWeaponsGameplay/Weapons/RadicalCoreAR/Audio/SoundWaves/Radical_CoreAR_Athena_Reload_Start_01}
 }
 {
  id=WavePlayer_1
  properties {soundWaveAssetPtr=/RadicalWeaponsGameplay/Weapons/RadicalCoreAR/Audio/SoundWaves/Radical_CoreAR_Athena_Reload_Start_02}
 }
 {
  id=WavePlayer_2
  properties {soundWaveAssetPtr=/RadicalWeaponsGameplay/Weapons/RadicalCoreAR/Audio/SoundWaves/Radical_CoreAR_Athena_Reload_Start_03}
 }
]
properties {
 attenuationSettings=&Fort_Weapon_Reload_Att
 bDebug=true
 bOverrideConcurrency=true
 maxDistance=2250
 soundClassObject=&Fort_WeaponFoley
 volumeMultiplier=0.4
}
