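# Niagara effect for the Radical AR reload, apparently the discarded magazine, as a burst-spawned mesh particle that falls and bounces.
# Spawn aligns the mesh to the owner's rotation with a yaw offset, then adds randomized downward world-space velocity and inherited velocity.
# Update applies gravity and a random spin, gates collision by normalized age, and shrinks the mesh to zero near the end of its short lifetime.
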NiagaraSystem /RadicalWeaponsGameplay/Weapons/RadicalCoreAR/FX/NS_RadicalAR_MagazineReload
dependencies {
 usedDataInterfaces [&NiagaraDataInterfaceVectorCurve]
 usedDynamicInputs []
 usedModules [&SystemState]
}
emitters {Particle_Emitter {
  emitterUpdateScript [
   &EmitterState()
   &SpawnBurst_Instantaneous(Use Spawn Probability=false, UseLoopCountLimit=false)
  ]
  particleSpawnScript [
   /SetVariables(Particles.MeshOrientation=&MultiplyQuaternion(
     Quaternion A=&MakeQuatFromAxisAngle(Axis=<0 0 0.5>, AngleInDegrees=90)
     Quaternion B=$Engine.Owner.Rotation
    )) {isSetParameters=true}
   /SetVariables(Particles.Acceleration:Vector3f=<0 0 -980>) {isSetParameters=true}
   /SetVariables(Particles.SpawnRotationVector=&RandomRangeVector(Minimum=<0 1 0>, Maximum=<0 -1 0>)) {isSetParameters=true}
   &AddVelocity(
    Velocity=&TransformVector(
     Vector=&RandomRangeVector(Minimum=<200 200 -400>, Maximum=<220 200 -400>)
     Source Space=::World
     Destination Space=::Simulation
    )
    Velocity Speed Scale=0.65
   )
   &InitializeParticle(Lifetime=1.2, Mesh Scale Mode=::NewEnumerator4, UsePositionOffset=false, Write Material Random=true)
   &InheritVelocity(Inherited Velocity Speed Limit=0, Speed Limit=false, UseSpeedThreshold=false)
   &ApplyInitialForces()
  ]
  particleUpdateScript [
   &ParticleState()
   &ScaleMeshSize(Scale Factor=&VectorFromCurve(
     VectorCurve=DataInterface() {propertyValues {XCurve {keys [<at 0 1>,<at 0.8 1>,<at 1 0>]}, YCurve {keys [<at 0 1>,<at 0.8 1>,<at 1 0>]}, ZCurve {keys [<at 0 1>,<at 0.8 1>,<at 1 0>]}}}
     Scale Curve=&VectorFromFloat(Value=1)
    ))
   &UpdateMeshOrientation(Rotation Vector=$Particles.SpawnRotationVector)
   &Color() {enabled=false}
   &AccelerationForce(Acceleration=$Particles.Acceleration)
   &Collision(
    Age Colliding Particles=false
    Collision Enabled=&SetBoolByFloatComparison(A=0.15, B=$Particles.NormalizedAge, Comparison Type=::NewEnumerator6)
    Friction=0.4
    Module.Control Rest State Via Surface Normal=false
    Module.EnableMaxCollisionCount=false
    Module.Randomize Collision Normal Vector=false
    Radius Calculation Type=::NewEnumerator1
    Restitution=0.5
   )
   &SolveForcesAndVelocity(Clamp Velocity=false, Limit Acceleration=false)
  ]
  rendererClasses [&NiagaraMeshRendererProperties]
  renderers [{rendererClass=&NiagaraMeshRendererProperties, rendererIndex=0}]
 }}
summary {
 systemName=NS_RadicalAR_MagazineReload
 emitters [{emitterName=Particle_Emitter, rendererClasses [&NiagaraMeshRendererProperties]}]
 userVariables []
}
