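# Physics asset for the Radical Core AR assault rifle mesh (Ch4 S2), giving the weapon a simple collision shape.
# A single body on the handle bone holds one capsule, offset and rotated from the bone origin.
# It has no constraints, so the asset defines no jointed or separately simulated parts.
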PhysicsAsset /RadicalWeaponsGameplay/Weapons/RadicalCoreAR/Meshes/Assault_Ch4_S2_Rifle_PhysicsAsset
bodies {handle [{shapeType=Capsule, length=48.9185, radius=12.7088, center=<20.105 1.2802 6.5178>, rotation=<rot 0 90.1457 -86.8825>}]}
constraints []
